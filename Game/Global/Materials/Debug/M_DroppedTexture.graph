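# Debug material in the default lit shading model, built around a single texture parameter.
# The graph holds only a BaseColor texture parameter defaulting to T_BlacksRGB, with no connection to a material output shown.
# Usage is flagged for skeletal meshes, morph targets and mesh deformers, with automatic usage setting in the editor turned off.
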Material /Game/Global/Materials/Debug/M_DroppedTexture
graph {
 properties {
  bAutomaticallySetUsageInEditor=false
  bUsedWithMeshDeformer=true
  bUsedWithMorphTargets=true
  bUsedWithSkeletalMesh=true
  usedShadingModels=Default Lit
 }
 nodes [TextureSampleParameter2D() {parameterName=BaseColor, texture=&T_BlacksRGB}]
}
